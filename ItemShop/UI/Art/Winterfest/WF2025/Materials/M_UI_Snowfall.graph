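# Translucent UI material for Winterfest 2025 Item Shop art that builds an animated snowfall effect.
# Panned, rotated dust and particle textures at foreground and background tilings are combined, with two scrolling blurred-noise samples modulating particle intensity.
# Bottom, left and right falloffs, blendable between A/B offsets by AnimateFalloff, mask the Snow result.
# The graph also carries Item Shop tile color gradient, rounded-mask and outline branches as named reroutes.
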Material /ItemShop/UI/Art/Winterfest/WF2025/Materials/M_UI_Snowfall
graph {
 properties {
  bAutomaticallySetUsageInEditor=false
  bUsedWithStaticMesh=false
  blendMode=BLEND_Translucent
  materialDomain=MD_UI
  usedShadingModels=Default Lit
 }
 nodes [
  n1 = LinearInterpolate(A=NamedRerouteUsage(), B=Divide(A=NamedRerouteUsage()) {constB=1}, Alpha=Multiply(A=NamedRerouteUsage()))
  n2 = Multiply(A=Divide(A=NamedRerouteUsage()) {constB=5}) {constB=0.25}
  Add(A=@n1, B=@n2)
  LinearInterpolate(A=@n1, B=NamedRerouteUsage(), Alpha=@n2)
  MaterialFunctionCall!(Input Alpha=Constant() {r=1}, bOverrideCornerRadius, Corner Radius Override) {materialFunction=/FortUILibrary/Systems/ItemShopTile/Materials/MaterialFunctions/MF_UI_ItemShop_RoundedCorners}
  n3 = MaterialFunctionCall() {materialFunction=&MF_UI_UVScaling}
  NamedRerouteDeclaration(None=@n3.Scaled to Fit) {name=UVScaling_Fit, nodeColor=<rgba 0.0781 0 1 1>}
  n4 = NamedRerouteUsage()
  n5 = NamedRerouteUsage()
  n6 = NamedRerouteUsage()
  n7 = Constant()
  n8 = Constant() {r=1}
  n9 = Constant() {r=-0.25}
  NamedRerouteDeclaration(None=LinearInterpolate(
    A=NamedRerouteUsage()
    B=MaterialFunctionCall!(Base=@n4, Blend=@n5) {materialFunction=/Engine/Functions/Engine_MaterialFunctions03/Blends/Blend_HardLight}.Result
    Alpha=MaterialFunctionCall(UVs=@n6, GradientStart=@n7, GradientEnd=@n8, RotationAngle=@n9) {materialFunction=&MF_UI_LinearGradient}.Result
   )) {name=OutlineGradient, nodeColor=<rgba 0 0 0 1>}
  NamedRerouteDeclaration(None=NamedRerouteUsage()) {name=Color2, nodeColor=<rgba 1 0.0007 0.2048 1>}
  n10 = MaterialFunctionCall() {materialFunction=/FortUILibrary/Systems/ItemShopTile/Materials/MaterialFunctions/MF_UI_ItemShop_BackgroundParams}
  NamedRerouteDeclaration(None=@n10.Color 2) {name=Secondary Color, nodeColor=<rgba 0 0 0 1>}
  NamedRerouteDeclaration(None=@n10.Color 3) {name=Accent Color, nodeColor=<rgba 0 0 0 1>}
  n11 = MaterialFunctionCall(UVs=NamedRerouteUsage(), GradientStart=Constant(), GradientEnd=Constant() {r=1.2}, RotationAngle=Constant() {r=0.25}) {materialFunction=&MF_UI_LinearGradient}
  NamedRerouteDeclaration(None=LinearInterpolate(A=NamedRerouteUsage(), B=NamedRerouteUsage(), Alpha=SmoothStep(Value=@n11.Result) {constMax=1.1})) {name=Gradient, nodeColor=<rgba 0 0 0 1>}
  NamedRerouteDeclaration(None=NamedRerouteUsage()) {name=Color3, nodeColor=<rgba 0 1 0.625 1>}
  n12 = CollectionParameter() {collection=/FortUILibrary/Systems/ItemShopTile/MPC_ItemShop, parameterName=CornerRadius}
  n13 = MaterialFunctionCall(
   Corner Radius=MaterialFunctionCall(R=@n12, G=@n12, B=@n12, A=@n12) {materialFunction=&AppendMany}.RGBA
   Radius in Pixels=Constant() {r=1}
   Outline Thickness=Constant() {r=6}
  ) {materialFunction=&MF_UI_CoreUI_RoundedMask}
  n14 = OneMinus(None=SmoothStep(Value=@n13.Outline SDF) {constMax=0.1})
  NamedRerouteDeclaration(None=LinearInterpolate(A=Multiply(A=@n14) {constB=0.2}, Alpha=@n13.Outline Mask)) {name=OutlineFrame, nodeColor=<rgba 0 0.5547 1 1>}
  NamedRerouteDeclaration(None=@n13.Shape Mask) {name=FlatBGMask, nodeColor=<rgba 0 0.7188 1 1>}
  n15 = TextureCoordinate()
  n16 = ScalarParameter() {group=Falloff, parameterName=AnimateFalloff}
  n17 = Add(
   A=OneMinus(None=ComponentMask(None=@n15) {g=true})
   B=LinearInterpolate(
    A=ScalarParameter() {defaultValue=-0.1, group=Falloff, parameterName=FalloffShift_Bottom_A}
    B=ScalarParameter() {defaultValue=-0.1, group=Falloff, parameterName=FalloffShift_Bottom_B}
    Alpha=@n16
   )
  )
  n18 = ComponentMask(None=Rotator(Coordinate=@n15, Time=ScalarParameter() {defaultValue=-0.6, group=Falloff, parameterName=FalloffLeft_Rotation}) {speed=6.2831}) {g=true}
  n19 = LinearInterpolate(
   A=ScalarParameter() {defaultValue=-0.3, group=Falloff, parameterName=FalloffShift_Left_A}
   B=ScalarParameter() {defaultValue=-0.3, group=Falloff, parameterName=FalloffShift_Left_B}
   Alpha=@n16
  )
  n20 = Saturate(None=Multiply(A=Add(A=@n18, B=@n19), B=ScalarParameter() {defaultValue=1, group=Falloff, parameterName=FalloffEdgeSoft_Left}))
  n21 = ComponentMask(None=Rotator(Coordinate=@n15, Time=ScalarParameter() {defaultValue=0.65, group=Falloff, parameterName="FalloffRight_Rotation "}) {speed=6.2831}) {g=true}
  n22 = LinearInterpolate(
   A=ScalarParameter() {defaultValue=-0.2, group=Falloff, parameterName=FalloffShift_Right_A}
   B=ScalarParameter() {group=Falloff, parameterName=FalloffShift_Right_B}
   Alpha=@n16
  )
  n23 = Saturate(None=Multiply(A=Add(A=@n21, B=@n22), B=ScalarParameter() {defaultValue=0.5, group=Falloff, parameterName=FalloffEdgeSoft_Right}))
  n24 = Saturate(None=DotProduct(A=Multiply(A=@n17), B=DotProduct(A=@n20, B=@n23)))
  n25 = ScalarParameter() {defaultValue=1.25, group=Dust, parameterName=DustTile_A}
  n26 = NamedRerouteUsage()
  n27 = Add(
   A=Multiply(A=@n26, B=ScalarParameter() {defaultValue=2.5, parameterName=GlobalTileForeground})
   B=Multiply(A=TextureCoordinate(), B=ScalarParameter() {defaultValue=0.25, parameterName=LocalizedUV})
  )
  n28 = TextureSample(UVs=Panner(Coordinate=Multiply(A=@n27, B=@n25)) {speedX=0.04, speedY=-0.08}) {samplerType=SAMPLERTYPE_Masks, texture=/OfferCatalog/Art/Majesty/T_UI_Dust}
  n29 = ScalarParameter() {defaultValue=1, group=Dust, parameterName=DustTile_B}
  n30 = Multiply(A=@n26, B=ScalarParameter() {defaultValue=8, parameterName=GlobalTileBackground})
  n31 = TextureSample(UVs=Panner(Coordinate=Multiply(A=@n30, B=@n29)) {speedX=0.08, speedY=-0.24}) {samplerType=SAMPLERTYPE_Masks, texture=&T_UI_Dust}
  n32 = ScalarParameter() {defaultValue=0.75, group=Particles, parameterName=ParticleScale}
  n33 = Time()
  n34 = ScalarParameter() {defaultValue=-0.075, group=Particles, parameterName=ParticleRotationBackground}
  n35 = Add(A=Rotator(Coordinate=@n30, Time=Multiply(A=@n33, B=@n34)) {centerX=0, centerY=0})
  n36 = Multiply(A=Time(), B=ScalarParameter() {defaultValue=0.1125, group=Particles, parameterName=ParticleTimeSpeedFactor})
  n37 = AppendVector(
   A=ScalarParameter() {defaultValue=0.14, group=Particles, parameterName=ParticlePanBackground_X}
   B=ScalarParameter() {defaultValue=-0.2, group=Particles, parameterName=ParticlePanBackground_Y}
  )
  n38 = TextureSample(UVs=Multiply(A=@n32, B=Panner(Coordinate=@n35, Time=@n36, Speed=@n37) {speedX=0.25, speedY=0.25})) {samplerType=SAMPLERTYPE_Masks, texture=/OfferCatalog/Art/Majesty/T_UI_Particles}
  n39 = Rotator(
   Coordinate=@n27
   Time=Multiply(A=Time(), B=ScalarParameter() {defaultValue=-0.125, group=Particles, parameterName=ParticleRotationForeground})
  ) {centerX=0, centerY=0}
  n40 = AppendVector(
   A=ScalarParameter() {defaultValue=0.07, group=Particles, parameterName=ParticlePanForeground_X}
   B=ScalarParameter() {defaultValue=-0.1, group=Particles, parameterName=ParticlePanForeground_Y}
  )
  n41 = TextureSample(UVs=Multiply(A=Panner(Coordinate=@n39, Time=@n36, Speed=@n40) {speedX=0.35, speedY=0.125}, B=@n32)) {samplerType=SAMPLERTYPE_Masks, texture=&T_UI_Particles}
  n42 = Panner(
   Coordinate=Multiply(A=@n26, B=ScalarParameter() {defaultValue=0.5, group=AlphaNoise, parameterName=NoiseMultA})
   Speed=AppendVector(
    A=ScalarParameter() {defaultValue=0.09, group=AlphaNoise, parameterName=NoisePanA_X}
    B=ScalarParameter() {defaultValue=-0.03, group=AlphaNoise, parameterName=NoisePanA_Y}
   )
  ) {speedX=0.09, speedY=-0.03}
  n43 = Panner(
   Coordinate=Multiply(A=@n26, B=ScalarParameter() {defaultValue=1.5, group=AlphaNoise, parameterName=NoiseMultB})
   Speed=AppendVector(
    A=ScalarParameter() {defaultValue=0.05, group=AlphaNoise, parameterName=NoisePanB_X}
    B=ScalarParameter() {defaultValue=-0.18, group=AlphaNoise, parameterName=NoisePanB_Y}
   )
  ) {speedX=0.05, speedY=-0.18}
  n44 = ScalarParameter() {defaultValue=5, group=AlphaNoise, parameterName=NoisePower}
  n45 = Power(
   Base=Add(
    A=TextureSample(UVs=@n42) {samplerType=SAMPLERTYPE_Grayscale, texture=&LowResBlurredNoise}.R
    B=TextureSample(UVs=@n43) {samplerType=SAMPLERTYPE_Grayscale, texture=&LowResBlurredNoise}.R
   )
   Exp=@n44
  )
  n46 = LinearInterpolate(
   A=ScalarParameter() {defaultValue=32, group=Particles, parameterName=ParticleIntensityLow}
   B=ScalarParameter() {defaultValue=128, group=Particles, parameterName=ParticleIntensityHigh}
   Alpha=Saturate(None=Multiply(A=@n45, B=@n44))
  )
  n47 = Multiply(A=Add(A=Multiply(A=@n38.B, B=@n41.R), B=Multiply(A=@n41.B, B=@n38.R)), B=@n46)
  n48 = DotProduct(A=Add(A=@n28.R, B=@n31.R), B=@n47)
  n49 = Saturate(None=Multiply(A=@n28.G, B=@n31.G))
  n50 = Multiply(A=Multiply(A=@n28.B, B=@n31.B), B=@n46)
  n51 = Multiply(A=Add(A=Add(A=@n48, B=@n47), B=Add(A=@n49, B=@n50)), B=ScalarParameter() {defaultValue=4, parameterName=GoldDustBoost})
  NamedRerouteDeclaration(None=LinearInterpolate(B=LinearInterpolate(B=@n24, Alpha=@n51), Alpha=@n51)) {name=Snow, nodeColor=<rgba 0.5703 1 0 1>}
  NamedRerouteDeclaration(None=@n10.Color 1) {name=Primary Color, nodeColor=<rgba 0 0 0 1>}
  NamedRerouteDeclaration(None=@n3.ScaleY) {name=Fill, nodeColor=<rgba 0 0.1328 1 1>}
  NamedRerouteDeclaration(None=NamedRerouteUsage()) {name=Color1, nodeColor=<rgba 1 0.0007 0.2048 1>}
 ]
}
